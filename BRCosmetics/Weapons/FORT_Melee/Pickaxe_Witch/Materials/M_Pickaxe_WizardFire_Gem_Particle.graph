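# Default Lit material for mesh particles, named for the gem of the Wizard Fire variant of the Witch pickaxe cosmetic.
# It samples the base Witch pickaxe diffuse, normal and specular mask textures alongside a fire-specific emissive texture.
# The emissive texture's RGB feeds the shared MF_FX_EmissiveAdjustments function, the only link shown between nodes.
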Material /BRCosmetics/Weapons/FORT_Melee/Pickaxe_Witch/Materials/M_Pickaxe_WizardFire_Gem_Particle
graph {
 properties {bUsedWithMeshParticles=true, pixelDepthOffsetMode=PDOM_Legacy, usedShadingModels=Default Lit}
 nodes [
  MaterialFunctionCall(Emissive=TextureSample() {texture=/BRCosmetics/Weapons/FORT_Melee/Pickaxe_Witch/Textures/Fire/T_Pickaxe_Wizard_Fire_E}.RGB) {materialFunction=&MF_FX_EmissiveAdjustments}
  TextureSample() {samplerType=SAMPLERTYPE_Normal, texture=/BRCosmetics/Weapons/FORT_Melee/Pickaxe_Witch/Textures/T_Pickaxe_Witch_N}
  TextureSample() {texture=/BRCosmetics/Weapons/FORT_Melee/Pickaxe_Witch/Textures/T_Pickaxe_Witch_D}
  TextureSample() {samplerType=SAMPLERTYPE_Masks, texture=/BRCosmetics/Weapons/FORT_Melee/Pickaxe_Witch/Textures/T_Pickaxe_Witch_S}
 ]
}
