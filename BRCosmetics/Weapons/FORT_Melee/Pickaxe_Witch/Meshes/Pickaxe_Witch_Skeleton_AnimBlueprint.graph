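# Animation Blueprint for the Witch pickaxe skeleton that makes its dynamic bones, apparently a crystal, bob and sway continuously.
# Each update, the EventGraph derives three sine offsets from game time at different frequencies, with a much larger amplitude on Z than on X and Y.
# The AnimGraph adds these translations to the input pose through chained Modify Bone nodes on dyn_axisX, dyn_axisY and dyn_axisZ, giving procedural motion without animation assets.
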AnimBlueprint /BRCosmetics/Weapons/FORT_Melee/Pickaxe_Witch/Meshes/Pickaxe_Witch_Skeleton_AnimBlueprint
target_skeleton=&/Pickaxe_Witch_Skeleton
graphs [
 AnimGraph {nodes [
   AnimGraphNode_SubInput_0 = Input Pose()
   n1 = "Transform (Modify) Bone - Bone: dyn_axisY"(
    ComponentPose="Transform (Modify) Bone - Bone: dyn_axisX"(
     ComponentPose=Local To Component(LocalPose=@AnimGraphNode_SubInput_0)
     Alpha=1
     Translation=Get crystalX() -> (crystalX:Vector)
     Rotation="0,0,0"
     Scale="1,1,1"
    ) {translationMode=BMM_Additive, alphaBoolBlend {blendOption=Linear}, boneToModify {boneName=dyn_axisX}}
    Alpha=1
    Translation=Get crystalY() -> (crystalY:Vector)
    Rotation="0,0,0"
    Scale="1,1,1"
   ) {translationMode=BMM_Additive, alphaBoolBlend {blendOption=Linear}, boneToModify {boneName=dyn_axisY}}
   n2 = Get crystalZ() -> (crystalZ:Vector)
   Output Pose(Result=Component To Local(ComponentPose="Transform (Modify) Bone - Bone: dyn_axisZ"(ComponentPose=@n1, Alpha=1, Translation=@n2, Rotation="0,0,0", Scale="1,1,1") {translationMode=BMM_Additive, alphaBoolBlend {blendOption=Linear}, boneToModify {boneName=dyn_axisZ}})) {layerGroup=None}
  ]}
 EventGraph {nodes [
   Event BlueprintUpdateAnimation()
   n1 = GetGameTimeInSeconds()
   n2 = "Sin (Degrees)"!(A:double=float * float(A=@n1, B=100)) -> (ReturnValue:double)
   > Set crystalX(crystalX:Vector=MakeVector(X=float * float(A=@n2, B=0.01), Y=0)) -> (Output_Get:Vector)
   n3 = "Sin (Degrees)"(A=float * float(A=@n1, B=125))
   > Set crystalY(crystalY:Vector=MakeVector(Y=float * float(A=@n3, B=0.01))) -> (Output_Get:Vector)
   n4 = "Sin (Degrees)"(A=float * float(A=@n1, B=150))
   > Set crystalZ(crystalZ:Vector=MakeVector(Y=0, Z=float * float(A=@n4, B=0.333))) -> (Output_Get:Vector)
   TryGetPawnOwner()
  ]}
]
variables [
 crystalX {type=Vector}
 crystalY {type=Vector}
 crystalZ {type=Vector}
]
